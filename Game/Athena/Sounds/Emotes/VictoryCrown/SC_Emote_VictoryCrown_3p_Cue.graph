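# Sound cue for the third-person audio of the Victory Crown emote in Athena.
# A mixer layers the fanfare wave with one of two crowd variations picked by an equally weighted random node.
# It uses emote attenuation, the third-person emote music class and emote music concurrency, and keeps playing when silent.
# It also has a send to a music audio analysis submix.
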SoundCue /Game/Athena/Sounds/Emotes/VictoryCrown/SC_Emote_VictoryCrown_3p_Cue
root=Mixer_1
nodes [
 {id=Mixer_1, children [WavePlayer_2,Random_0], properties {inputVolume [1,1]}}
 {id=Random_0, children [WavePlayer_3,WavePlayer_4], properties {weights [1,1]}}
 {id=WavePlayer_2, properties {soundWaveAssetPtr=&/sfx_Emote_Victorycrown_Fanfare}}
 {id=WavePlayer_3, properties {soundWaveAssetPtr=&/sfx_Emote_Victorycrown_Fanfare_Crowd_01}}
 {id=WavePlayer_4, properties {soundWaveAssetPtr=&/sfx_Emote_Victorycrown_Fanfare_Crowd_02}}
]
properties {
 attenuationSettings=&Fort_Emotes_Athena_Att
 maxDistance=1600
 soundClassObject=/Game/Sounds/FortSoundClassesAndModes/Fort_Music_Emote_3P
 virtualizationMode=PlayWhenSilent
 volumeMultiplier=1
 concurrencySet [/Game/Sounds/Concurrency_Rules/Athena_EmoteMusic_Conc]
 soundSubmixSends [{minSendLevel=0, soundSubmix=&Fortnite_AudioAnalysis_Music}]
}
